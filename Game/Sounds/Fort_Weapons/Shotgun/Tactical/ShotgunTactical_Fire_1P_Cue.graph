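# Sound cue for the first-person fire sound of the tactical shotgun.
# A Random node picks one of four equally weighted close-range fire recordings and feeds it through a Modulator, left at default settings, to the output.
# Playback uses the first-person ranged weapon fire sound class, a first-person Athena projectile attenuation asset and a shared first-person gunfire concurrency set.
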SoundCue /Game/Sounds/Fort_Weapons/Shotgun/Tactical/ShotgunTactical_Fire_1P_Cue
root=Modulator_0
nodes [
 {id=Modulator_0, children [Random_1]}
 {id=Random_1, children [WavePlayer_2,WavePlayer_3,WavePlayer_4,WavePlayer_5], properties {weights [1,1,1,1]}}
 {id=WavePlayer_2, properties {soundWaveAssetPtr=&/ShotgunTactical_Fire_Close_03}}
 {id=WavePlayer_3, properties {soundWaveAssetPtr=&/ShotgunTactical_Fire_Close_01}}
 {id=WavePlayer_4, properties {soundWaveAssetPtr=&/ShotgunTactical_Fire_Close_02}}
 {id=WavePlayer_5, properties {soundWaveAssetPtr=&/ShotgunTactical_Fire_Close_04}}
]
properties {
 attenuationSettings=&Fort_Weapon_Projectile_FirstPerson_Athena_Att
 maxDistance=2097152
 soundClassObject=&Fort_WeaponRanged_Fire_1P_PSM
 attenuationOverrides {falloffDistance=5000, attenuationShapeExtents {x=1000}}
 concurrencyOverrides {bLimitToOwner=true, maxCount=3, resolutionRule=StopOldest}
 concurrencySet [&Weapon_Gunfire_FirstPerson_Conc]
}
