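# Sound cue that plays the pickup sound for the tactical shotgun when it is collected as loot.
# A single wave player feeds the root directly, with lowered volume and the weapon pickup attenuation in the loot collect sound class.
# It references a shared oldest-based pickup concurrency set alongside a one-voice prevent-new override.
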SoundCue /Game/Sounds/Fort_Weapons/Shotgun/Tactical/ShotgunTactical_Pickup_Cue
root=WavePlayer_2
nodes [{id=WavePlayer_2, properties {soundWaveAssetPtr=&/ShotgunTactical_Pickup_01}}]
properties {
 attenuationSettings=&Fort_Weapon_Pickup_Att
 maxDistance=3500
 soundClassObject=&Fort_Loot_Collect
 volumeMultiplier=0.2308
 concurrencyOverrides {maxCount=1, resolutionRule=PreventNew}
 concurrencySet [&PickupMaximum_Oldest]
}
