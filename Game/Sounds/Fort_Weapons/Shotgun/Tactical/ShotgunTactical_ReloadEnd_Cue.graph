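# Sound cue for the tactical shotgun's reload-end foley in Fort weapons.
# A single wave player feeds a short fixed delay, then a Modulator at default settings for slight variation.
# It plays through the weapon foley sound class at reduced volume with reload attenuation.
# Overridden concurrency limits instances per owner and stops the oldest when exceeded.
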SoundCue /Game/Sounds/Fort_Weapons/Shotgun/Tactical/ShotgunTactical_ReloadEnd_Cue
root=Modulator_14
nodes [
 {id=Delay_0, children [WavePlayer_0], properties {delayMax=0.1, delayMin=0.1}}
 {id=Modulator_14, children [Delay_0]}
 {id=WavePlayer_0, properties {soundWaveAssetPtr=&/ShotgunTactical_ReloadEnd_01}}
]
properties {
 attenuationSettings=&Fort_Weapon_Reload_Att
 bOverrideConcurrency=true
 maxDistance=2250
 soundClassObject=&Fort_WeaponFoley
 volumeMultiplier=0.401
 attenuationOverrides {falloffDistance=1600}
 concurrencyOverrides {bLimitToOwner=true, maxCount=2, resolutionRule=StopOldest}
}
