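# Sound cue for the tactical shotgun's shell-insert reload foley.
# A random node picks one of four equally weighted insert recordings, which feeds a modulator that slightly varies pitch on each play.
# It plays under the weapon foley sound class with reload attenuation and lowered volume.
# Its own concurrency override uses a prevent-new rule, so extra instances are refused once the limit is reached.
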SoundCue /Game/Sounds/Fort_Weapons/Shotgun/Tactical/ShotgunTactical_ReloadInsert_Cue
root=Modulator_13
nodes [
 {id=Modulator_13, children [Random_0], properties {pitchMax=1.1, pitchMin=0.9}}
 {id=Random_0, children [WavePlayer_0,WavePlayer_2,WavePlayer_6,WavePlayer_7], properties {weights [1,1,1,1]}}
 {id=WavePlayer_0, properties {soundWaveAssetPtr=&/ShotgunTactical_ReloadInsert_01}}
 {id=WavePlayer_2, properties {soundWaveAssetPtr=&/ShotgunTactical_ReloadInsert_03}}
 {id=WavePlayer_6, properties {soundWaveAssetPtr=&/ShotgunTactical_ReloadInsert_04}}
 {id=WavePlayer_7, properties {soundWaveAssetPtr=&/ShotgunTactical_ReloadInsert_05}}
]
properties {
 attenuationSettings=&Fort_Weapon_Reload_Att
 bOverrideConcurrency=true
 maxDistance=2250
 soundClassObject=&Fort_WeaponFoley
 volumeMultiplier=0.601
 attenuationOverrides {falloffDistance=1600}
 concurrencyOverrides {resolutionRule=PreventNew}
}
